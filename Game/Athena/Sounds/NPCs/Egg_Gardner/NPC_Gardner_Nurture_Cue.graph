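# Sound cue for the Egg Gardner NPC's nurture sound, played in Athena.
# A single wave player plays the Egg_Planter_Nurture_01 wave directly, with no randomization or modulation.
# The cue routes to the gadgets sound class at reduced volume and borrows the NPC teleport attenuation settings.
# A concurrency override caps how many instances can play at once.
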SoundCue /Game/Athena/Sounds/NPCs/Egg_Gardner/NPC_Gardner_Nurture_Cue
root=WavePlayer_3
nodes [{id=WavePlayer_3, properties {soundWaveAssetPtr=&/Egg_Planter_Nurture_01}}]
properties {
 attenuationSettings=/Game/Athena/Sounds/NPCs/NPC_Teleport_Att
 bOverrideConcurrency=true
 maxDistance=8500
 soundClassObject=&Fort_Gadgets
 volumeMultiplier=0.5
 concurrencyOverrides {maxCount=2}
}
